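# Physics asset for the M_MED_Bone_Snake male medium player body cosmetic, covering its dangling cloth, scarf, horns and shoelaces.
# Capsules on pelvis, spine and thighs, plus small calf spheres, form the body core that the accessories attach to.
# Most accessory chains hang from tiny base bodies with fully free constraints, while top cloth and shoelaces attach directly.
# Accessory joints use tight swing and twist limits, with flap chains mostly bending along one swing axis.
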PhysicsAsset /BRCosmetics/Characters/Player/Male/Medium/Bodies/M_MED_Bone_Snake/Meshes/M_MED_Bone_Snake_Physics
bodies {
 calf_twist_01_l [{shapeType=Sphere, radius=1}]
 calf_twist_01_r [{shapeType=Sphere, radius=1}]
 dyn_clothBottomA_a [{shapeType=Box, extentX=5.5, extentY=3.5, extentZ=7.5, center=<2.8025 0 0>}]
 dyn_clothBottomA_a_base [{shapeType=Capsule, length=1, radius=1}]
 dyn_clothBottomA_b [{shapeType=Box, extentX=7.5, extentY=4, extentZ=8, center=<3.6626 0.0001 0>}]
 dyn_clothBottomA_c [{shapeType=Box, extentX=9.5, extentY=4, extentZ=8, center=<4.8798 0 0.0001>}]
 dyn_clothBottomB_a [{shapeType=Box, extentX=5, extentY=2.5, extentZ=8, center=<2.7753 0 0>}]
 dyn_clothBottomB_a_base [{shapeType=Capsule, length=1, radius=1}]
 dyn_clothBottomB_b [{shapeType=Box, extentX=5.5, extentY=2.5, extentZ=8, center=<2.8121 0 0>}]
 dyn_clothBottomB_c [{shapeType=Box, extentX=6, extentY=2.5, extentZ=8, center=<2.8678 0 0>}]
 dyn_clothBottomC_a [{shapeType=Box, extentX=7.5, extentY=9.5, extentZ=3.5, center=<3.8134 -0.0001 0>}]
 dyn_clothBottomC_a_base [{shapeType=Capsule, length=1, radius=1}]
 dyn_clothBottomC_b [{shapeType=Box, extentX=10, extentY=10, extentZ=3.5, center=<5.0815 -0.0001 0>}]
 dyn_clothBottomD_a [{shapeType=Box, extentX=8.5, extentY=21.5, extentZ=2.5, center=<4.5345 0 0>}]
 dyn_clothBottomD_a_base [{shapeType=Capsule, length=1, radius=1}]
 dyn_clothBottomD_b [{shapeType=Box, extentX=6.5, extentY=17.5, extentZ=2.5, center=<3.72 0 0>}]
 dyn_clothBottomD_c [{shapeType=Box, extentX=12, extentY=15.5, extentZ=2.5, center=<6.3236 0 0>}]
 dyn_clothBottomSide_a [{shapeType=Box, extentX=9.5, extentY=9.5, extentZ=3.5, center=<5.2938 0.0001 -0.0001>}]
 dyn_clothBottomSide_a_base [{shapeType=Capsule, length=1, radius=1}]
 dyn_clothTop_a [{shapeType=Box, extentX=2, extentY=4.5, extentZ=5, center=<0.5429 0 0>}]
 dyn_clothTop_b [{shapeType=Box, extentX=6.5, extentY=9.5, extentZ=2, center=<-1.746 0 0.0001>, rotation=<rot 0 0 -9.2136>}]
 dyn_clothTop_c [{shapeType=Box, extentX=8.5, extentY=10.5, extentZ=3, center=<-1.1734 0 0>}]
 dyn_clothTop_d [{shapeType=Box, extentX=8, extentY=2, extentZ=8.5, center=<-1.7452 0 0>, rotation=<rot 0 1.8826 0>}]
 dyn_clothTop_e [{shapeType=Box, extentX=6.5, extentY=1.5, extentZ=8.5, center=<-1.7452 0 0>, rotation=<rot 0 0 9.4188>}]
 dyn_loleg_horn_a_base_l [{shapeType=Capsule, length=1, radius=1}]
 dyn_loleg_horn_a_l [{shapeType=Capsule, length=5.25, radius=2, center=<4.7346 0 0.0001>, rotation=<rot 90 0 0>}]
 dyn_loleg_horn_b_l [{shapeType=Capsule, length=4, radius=1.5, center=<3.2347 0 0>, rotation=<rot 90 0 0>}]
 dyn_roleg_horn_a_base_r [{shapeType=Capsule, length=1, radius=1}]
 dyn_roleg_horn_a_r [{shapeType=Capsule, length=5.75, radius=2.25, center=<-5.1391 0 0>, rotation=<rot 90 0 0>}]
 dyn_roleg_horn_b_r [{shapeType=Capsule, length=4, radius=1.5, center=<-3.2864 0 0.0001>, rotation=<rot 90 0 0>}]
 dyn_scarf_a_l [{shapeType=Capsule, length=4.2768, radius=1.3933, center=<4.4367 -0.0001 0>, rotation=<rot 90 0 0>}]
 dyn_scarf_b_l [{shapeType=Capsule, length=3.3526, radius=1.5905, center=<3.9522 0 0>, rotation=<rot 90 0 0>}]
 dyn_scarf_back_a [{shapeType=Box, extentX=7, extentY=2, extentZ=22, center=<3.6037 0 0>}]
 dyn_scarf_back_a_base [{shapeType=Capsule, length=1, radius=1}]
 dyn_scarf_back_b [{shapeType=Box, extentX=8, extentY=2, extentZ=22, center=<4.0942 0 0>}]
 dyn_scarf_base_base_l [{shapeType=Capsule, length=1, radius=1}]
 dyn_scarf_base_l [{shapeType=Capsule, length=3.4681, radius=2.4569, center=<3.9697 0 0>, rotation=<rot 90 0 0>}]
 dyn_scarf_front_a [{shapeType=Box, extentX=7, extentY=2.5, extentZ=23.5, center=<3.8777 0 0>}]
 dyn_scarf_front_a_base [{shapeType=Capsule, length=1, radius=1}]
 dyn_scarf_front_b [{shapeType=Box, extentX=12, extentY=2.5, extentZ=23.5, center=<6.5895 0 0>}]
 dyn_shoelace_a_l [{shapeType=Capsule, length=2, radius=1, center=<0 0 1.6287>}]
 dyn_shoelace_a_r [{shapeType=Capsule, length=2, radius=1, center=<0 0 -1.6287>, rotation=<rot 0 0 180>}]
 dyn_shoelace_b_l [{shapeType=Capsule, length=2, radius=1, center=<0 0 -1.305>}]
 dyn_shoelace_b_r [{shapeType=Capsule, length=2, radius=1, center=<0 0 1.305>, rotation=<rot 0 0 180>}]
 dyn_torsoHornA_base_l [{shapeType=Sphere, radius=1}]
 dyn_torsoHornA_l [{shapeType=Capsule, length=7.25, radius=1.75, center=<1.7906 0 0>, rotation=<rot 90 0 0>}]
 dyn_torsoHornB_base_l [{shapeType=Sphere, radius=1}]
 dyn_torsoHornB_l [{shapeType=Capsule, length=16.5, radius=0.75, center=<5.3872 0.9152 0>, rotation=<rot 90 -15.293 -16.0675>}]
 dyn_torsoHornC_base_l [{shapeType=Sphere, radius=1}]
 dyn_torsoHornC_l [{shapeType=Capsule, length=14.75, radius=1, center=<3.9269 0.7568 0.0945>, rotation=<rot 81.2104 -174.933 177.2741>}]
 dyn_torsoHornD_base_l [{shapeType=Sphere, radius=1}]
 dyn_torsoHornD_l [{shapeType=Capsule, length=22.5, radius=1.75, center=<7.9925 2.4732 2.5986>, rotation=<rot 74.0989 179.9999 169.4405>}]
 dyn_torsoHornE_base_l [{shapeType=Sphere, radius=1}]
 dyn_torsoHornE_l [{shapeType=Capsule, length=12.75, radius=1.5, center=<3.5702 1.6833 1.9329>, rotation=<rot 77.2673 -177.7072 179.8009>}]
 pelvis [
  {shapeType=Capsule, length=7.5, radius=9.5, center=<-1.4739 -1.2311 0.0004>, rotation=<rot 0 0 -0.1992>}
  {shapeType=Capsule, length=7.25, radius=9.75, center=<-11.3184 0.0816 0.0001>}
 ]
 spine_02 [{shapeType=Capsule, length=6.25, radius=8, center=<0 1.2188 0>, rotation=<rot 0 0 -0.1992>}]
 spine_04 [{shapeType=Capsule, length=9.75, radius=11, center=<0 0.0994 0>}]
 spine_05 [{shapeType=Capsule, length=9.5, radius=10.75, center=<0 0.8479 0>}]
 thigh_l [{shapeType=Capsule, length=27.75, radius=8, center=<-15.0048 0.0001 0.4086>, rotation=<rot 84.6445 -180 -176.8293>}]
 thigh_r [{shapeType=Capsule, length=26.75, radius=9, center=<15.0048 -0.0001 -0.4086>, rotation=<rot 84.6444 -180 3.1707>}]
}
constraints [
 {
  bone1Name=dyn_scarf_front_a_base
  bone2Name=spine_05
  swing1LimitDegrees=45
  swing1Motion=Free
  swing2LimitDegrees=45
  swing2Motion=Free
  twistLimitDegrees=45
  twistMotion=Free
 }
 {
  bone1Name=dyn_scarf_back_a_base
  bone2Name=spine_05
  swing1LimitDegrees=45
  swing1Motion=Free
  swing2LimitDegrees=45
  swing2Motion=Free
  twistLimitDegrees=45
  twistMotion=Free
 }
 {
  bone1Name=spine_05
  bone2Name=spine_04
  swing1LimitDegrees=45
  swing1Motion=Free
  swing2LimitDegrees=45
  swing2Motion=Free
  twistLimitDegrees=45
  twistMotion=Free
 }
 {bone1Name=dyn_scarf_front_a, bone2Name=dyn_scarf_front_a_base, swing1LimitDegrees=25, swing2LimitDegrees=2, twistLimitDegrees=2}
 {bone1Name=dyn_scarf_front_b, bone2Name=dyn_scarf_front_a, swing1LimitDegrees=30, swing2LimitDegrees=2, twistLimitDegrees=2}
 {bone1Name=dyn_scarf_back_a, bone2Name=dyn_scarf_back_a_base, swing1LimitDegrees=20, swing2LimitDegrees=2, twistLimitDegrees=2}
 {bone1Name=dyn_scarf_back_b, bone2Name=dyn_scarf_back_a, swing1LimitDegrees=20, swing2LimitDegrees=2, twistLimitDegrees=2}
 {
  bone1Name=spine_04
  bone2Name=spine_02
  swing1LimitDegrees=45
  swing1Motion=Free
  swing2LimitDegrees=45
  swing2Motion=Free
  twistLimitDegrees=45
  twistMotion=Free
 }
 {bone1Name=dyn_clothTop_e, bone2Name=spine_02, swing1LimitDegrees=20, swing2LimitDegrees=20, twistLimitDegrees=15}
 {bone1Name=dyn_clothTop_d, bone2Name=spine_02, swing1LimitDegrees=20, swing2LimitDegrees=20, twistLimitDegrees=15}
 {bone1Name=dyn_clothTop_c, bone2Name=spine_02, swing1LimitDegrees=20, swing2LimitDegrees=20, twistLimitDegrees=15}
 {bone1Name=dyn_clothTop_b, bone2Name=spine_02, swing1LimitDegrees=20, swing2LimitDegrees=20, twistLimitDegrees=15}
 {bone1Name=dyn_clothTop_a, bone2Name=spine_02, swing1LimitDegrees=15, swing2LimitDegrees=15, twistLimitDegrees=15}
 {bone1Name=dyn_clothBottomD_a, bone2Name=dyn_clothBottomD_a_base, swing1LimitDegrees=5, swing2LimitDegrees=35, twistLimitDegrees=5}
 {bone1Name=dyn_clothBottomD_b, bone2Name=dyn_clothBottomD_a, swing1LimitDegrees=5, swing2LimitDegrees=35, twistLimitDegrees=5}
 {
  bone1Name=spine_02
  bone2Name=pelvis
  swing1LimitDegrees=45
  swing1Motion=Free
  swing2LimitDegrees=45
  swing2Motion=Free
  twistLimitDegrees=45
  twistMotion=Free
 }
 {
  bone1Name=dyn_clothBottomA_a_base
  bone2Name=pelvis
  swing1LimitDegrees=45
  swing1Motion=Free
  swing2LimitDegrees=45
  swing2Motion=Free
  twistLimitDegrees=45
  twistMotion=Free
 }
 {
  bone1Name=dyn_clothBottomB_a_base
  bone2Name=pelvis
  swing1LimitDegrees=45
  swing1Motion=Free
  swing2LimitDegrees=45
  swing2Motion=Free
  twistLimitDegrees=45
  twistMotion=Free
 }
 {
  bone1Name=dyn_clothBottomD_a_base
  bone2Name=pelvis
  swing1LimitDegrees=45
  swing1Motion=Free
  swing2LimitDegrees=45
  swing2Motion=Free
  twistLimitDegrees=45
  twistMotion=Free
 }
 {
  bone1Name=dyn_clothBottomC_a_base
  bone2Name=pelvis
  swing1LimitDegrees=45
  swing1Motion=Free
  swing2LimitDegrees=45
  swing2Motion=Free
  twistLimitDegrees=45
  twistMotion=Free
 }
 {
  bone1Name=thigh_l
  bone2Name=pelvis
  swing1LimitDegrees=45
  swing1Motion=Free
  swing2LimitDegrees=45
  swing2Motion=Free
  twistLimitDegrees=45
  twistMotion=Free
 }
 {
  bone1Name=thigh_r
  bone2Name=pelvis
  swing1LimitDegrees=45
  swing1Motion=Free
  swing2LimitDegrees=45
  swing2Motion=Free
  twistLimitDegrees=45
  twistMotion=Free
 }
 {bone1Name=dyn_clothBottomC_a, bone2Name=dyn_clothBottomC_a_base, swing1LimitDegrees=5, swing2LimitDegrees=40, twistLimitDegrees=5}
 {bone1Name=dyn_clothBottomB_a, bone2Name=dyn_clothBottomB_a_base, swing1LimitDegrees=20, swing2LimitDegrees=5, twistLimitDegrees=5}
 {bone1Name=dyn_clothBottomB_b, bone2Name=dyn_clothBottomB_a, swing1LimitDegrees=20, swing2LimitDegrees=5, twistLimitDegrees=5}
 {bone1Name=dyn_clothBottomB_c, bone2Name=dyn_clothBottomB_b, swing1LimitDegrees=20, swing2LimitDegrees=5, twistLimitDegrees=5}
 {bone1Name=dyn_clothBottomA_a, bone2Name=dyn_clothBottomA_a_base, swing1LimitDegrees=15, swing2LimitDegrees=5, twistLimitDegrees=5}
 {bone1Name=dyn_clothBottomA_b, bone2Name=dyn_clothBottomA_a, swing1LimitDegrees=20, swing2LimitDegrees=5, twistLimitDegrees=5}
 {bone1Name=dyn_clothBottomA_c, bone2Name=dyn_clothBottomA_b, swing1LimitDegrees=20, swing2LimitDegrees=5, twistLimitDegrees=5}
 {
  bone1Name=dyn_torsoHornA_base_l
  bone2Name=pelvis
  swing1LimitDegrees=45
  swing1Motion=Free
  swing2LimitDegrees=45
  swing2Motion=Free
  twistLimitDegrees=45
  twistMotion=Free
 }
 {
  bone1Name=dyn_torsoHornB_base_l
  bone2Name=pelvis
  swing1LimitDegrees=45
  swing1Motion=Free
  swing2LimitDegrees=45
  swing2Motion=Free
  twistLimitDegrees=45
  twistMotion=Free
 }
 {
  bone1Name=dyn_torsoHornC_base_l
  bone2Name=pelvis
  swing1LimitDegrees=45
  swing1Motion=Free
  swing2LimitDegrees=45
  swing2Motion=Free
  twistLimitDegrees=45
  twistMotion=Free
 }
 {
  bone1Name=dyn_torsoHornD_base_l
  bone2Name=pelvis
  swing1LimitDegrees=45
  swing1Motion=Free
  swing2LimitDegrees=45
  swing2Motion=Free
  twistLimitDegrees=45
  twistMotion=Free
 }
 {
  bone1Name=dyn_torsoHornE_base_l
  bone2Name=pelvis
  swing1LimitDegrees=45
  swing1Motion=Free
  swing2LimitDegrees=45
  swing2Motion=Free
  twistLimitDegrees=45
  twistMotion=Free
 }
 {bone1Name=dyn_torsoHornA_l, bone2Name=dyn_torsoHornA_base_l, swing1LimitDegrees=15, swing2LimitDegrees=15, twistLimitDegrees=5}
 {bone1Name=dyn_torsoHornB_l, bone2Name=dyn_torsoHornB_base_l, swing1LimitDegrees=15, swing2LimitDegrees=15, twistLimitDegrees=5}
 {bone1Name=dyn_torsoHornC_l, bone2Name=dyn_torsoHornC_base_l, swing1LimitDegrees=15, swing2LimitDegrees=15, twistLimitDegrees=5}
 {bone1Name=dyn_torsoHornD_l, bone2Name=dyn_torsoHornD_base_l, swing1LimitDegrees=15, swing2LimitDegrees=15, twistLimitDegrees=5}
 {bone1Name=dyn_torsoHornE_l, bone2Name=dyn_torsoHornE_base_l, swing1LimitDegrees=15, swing2LimitDegrees=15, twistLimitDegrees=5}
 {
  bone1Name=calf_twist_01_l
  bone2Name=thigh_l
  swing1LimitDegrees=45
  swing1Motion=Free
  swing2LimitDegrees=45
  swing2Motion=Free
  twistLimitDegrees=45
  twistMotion=Free
 }
 {bone1Name=dyn_shoelace_a_l, bone2Name=calf_twist_01_l, swing1LimitDegrees=15, swing2LimitDegrees=15, twistLimitDegrees=15}
 {bone1Name=dyn_shoelace_b_l, bone2Name=calf_twist_01_l, swing1LimitDegrees=15, swing2LimitDegrees=15, twistLimitDegrees=15}
 {bone1Name=dyn_shoelace_a_r, bone2Name=calf_twist_01_r, swing1LimitDegrees=15, swing2LimitDegrees=15, twistLimitDegrees=15}
 {bone1Name=dyn_shoelace_b_r, bone2Name=calf_twist_01_r, swing1LimitDegrees=15, swing2LimitDegrees=15, twistLimitDegrees=15}
 {
  bone1Name=calf_twist_01_r
  bone2Name=thigh_r
  swing1LimitDegrees=45
  swing1Motion=Free
  swing2LimitDegrees=45
  swing2Motion=Free
  twistLimitDegrees=45
  twistMotion=Free
 }
 {
  bone1Name=dyn_scarf_base_base_l
  bone2Name=spine_05
  swing1LimitDegrees=45
  swing1Motion=Free
  swing2LimitDegrees=45
  swing2Motion=Free
  twistLimitDegrees=45
  twistMotion=Free
 }
 {bone1Name=dyn_scarf_base_l, bone2Name=dyn_scarf_base_base_l, swing1LimitDegrees=5, swing2LimitDegrees=5, twistLimitDegrees=5}
 {bone1Name=dyn_scarf_a_l, bone2Name=dyn_scarf_base_l, swing1LimitDegrees=5, swing2LimitDegrees=5, twistLimitDegrees=5}
 {bone1Name=dyn_scarf_b_l, bone2Name=dyn_scarf_base_l, swing1LimitDegrees=5, swing2LimitDegrees=5, twistLimitDegrees=5}
 {bone1Name=dyn_clothBottomC_b, bone2Name=dyn_clothBottomC_a, swing1LimitDegrees=10, swing2LimitDegrees=40, twistLimitDegrees=5}
 {
  bone1Name=dyn_clothBottomSide_a_base
  bone2Name=pelvis
  swing1LimitDegrees=45
  swing1Motion=Free
  swing2LimitDegrees=45
  swing2Motion=Free
  twistLimitDegrees=45
  twistMotion=Free
 }
 {bone1Name=dyn_clothBottomSide_a, bone2Name=dyn_clothBottomSide_a_base, swing1LimitDegrees=5, swing2LimitDegrees=45, twistLimitDegrees=5}
 {
  bone1Name=dyn_loleg_horn_a_base_l
  bone2Name=thigh_l
  swing1LimitDegrees=45
  swing1Motion=Free
  swing2LimitDegrees=45
  swing2Motion=Free
  twistLimitDegrees=45
  twistMotion=Free
 }
 {bone1Name=dyn_loleg_horn_a_l, bone2Name=dyn_loleg_horn_a_base_l, swing1LimitDegrees=5, swing2LimitDegrees=5, twistLimitDegrees=5}
 {bone1Name=dyn_loleg_horn_b_l, bone2Name=dyn_loleg_horn_a_l, swing1LimitDegrees=15, swing2LimitDegrees=15, twistLimitDegrees=5}
 {
  bone1Name=dyn_roleg_horn_a_base_r
  bone2Name=thigh_r
  swing1LimitDegrees=45
  swing1Motion=Free
  swing2LimitDegrees=45
  swing2Motion=Free
  twistLimitDegrees=45
  twistMotion=Free
 }
 {bone1Name=dyn_roleg_horn_a_r, bone2Name=dyn_roleg_horn_a_base_r, swing1LimitDegrees=5, swing2LimitDegrees=5, twistLimitDegrees=5}
 {bone1Name=dyn_roleg_horn_b_r, bone2Name=dyn_roleg_horn_a_r, swing1LimitDegrees=15, swing2LimitDegrees=15, twistLimitDegrees=5}
 {bone1Name=dyn_clothBottomD_c, bone2Name=dyn_clothBottomD_b, swing1LimitDegrees=5, swing2LimitDegrees=35, twistLimitDegrees=5}
]
