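# Animation Blueprint for the Bone Snake medium male body cosmetic, copying the attached parent's pose and adding secondary motion.
# Weighted constraints place thigh, kneepad, horn, scarf and cloth bones, with corrective offsets driven by leg and shoulder alphas.
# AnimDynamics swings upper cloth, a RigidBody pass follows, then IsBackpackOn blends back-scarf constraints before fixed offsets.
# The hornBelt_a AnimDynamics and front-scarf constraint nodes are disconnected from the output.
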AnimBlueprint /BRCosmetics/Characters/Player/Male/Medium/Bodies/M_MED_Bone_Snake/Meshes/M_MED_Bone_Snake_AnimBP
target_skeleton=&/M_MED_Bone_Snake_Skeleton
graphs [
 AnimGraph {nodes [
   Anim Dynamics - Bone: dyn_hornBelt_a(Alpha=1, ExternalForce="0,0,0") {
    angularDampingOverride=1
    angularSpringConstant=1500
    bAngularSpring=true
    bGravityOverrideInSimSpace=true
    bOverrideAngularDamping=true
    gravityScale=0.35
    simulationSpace=RootRelative
    alphaBoolBlend {blendOption=Linear}
    boundBone {boneName=dyn_hornBelt_a}
    physicsBodyDefinitions [{
      sphereCollisionRadius=0
      boundBone {boneName=dyn_hornBelt_a}
      boxExtents=<10 20 20>
      constraintSetup {angularLimitsMax=<0 0 0>, angularLimitsMin=<0 -10 0>, angularTarget=<0 0 0>}
      localJointOffset=<10 0 0>
     }]
   }
   n1 = "Constraint - dyn_scarf_front_a (PPPPPP)"!(
    ComponentPose
    Alpha=1
    ConstraintWeights_0=0.0903
    ConstraintWeights_1:float=0.0629
    ConstraintWeights_2:float=0.1462
    ConstraintWeights_3:float=0.3202
    ConstraintWeights_4:float=0.2584
    ConstraintWeights_5:float=0.122
   ) -> (Pose) {
    alphaBoolBlend {blendOption=Linear}
    boneToModify {boneName=dyn_scarf_front_a}
    constraintSetup [
     {targetBone {boneName=clavicle_l}}
     {targetBone {boneName=spine_02}}
     {targetBone {boneName=spine_03}}
     {targetBone {boneName=spine_04}}
     {targetBone {boneName=spine_05}}
     {targetBone {boneName=clavicle_r}}
    ]
    constraintWeights [1,1,1,1,1,1]
   }
   "Constraint - dyn_scarf_front_b (PPPPPPP)"!(
    ComponentPose=@n1
    Alpha=1
    ConstraintWeights_0=0.0396
    ConstraintWeights_1:float=0.0239
    ConstraintWeights_2:float=0.1377
    ConstraintWeights_3:float=0.2714
    ConstraintWeights_4:float=0.3716
    ConstraintWeights_5:float=0.0833
    ConstraintWeights_6:float=0.0725
   ) -> (Pose) {
    alphaBoolBlend {blendOption=Linear}
    boneToModify {boneName=dyn_scarf_front_b}
    constraintSetup [
     {targetBone {boneName=clavicle_l}}
     {targetBone {boneName=spine_01}}
     {targetBone {boneName=spine_02}}
     {targetBone {boneName=spine_03}}
     {targetBone {boneName=spine_04}}
     {targetBone {boneName=spine_05}}
     {targetBone {boneName=clavicle_r}}
    ]
    constraintWeights [1,1,1,1,1,1,1]
   }
   n2 = "Constraint - thigh_volume_l (R)"(ComponentPose=Local To Component(LocalPose=Copy Pose From Mesh() {bUseAttachedParent=true}), Alpha=0.1, ConstraintWeights_0=1) {
    alphaBoolBlend {blendOption=Linear}
    boneToModify {boneName=thigh_volume_l}
    constraintSetup [{transformType=Rotation, targetBone {boneName=thigh_l}}]
    constraintWeights [1]
   }
   n3 = "Constraint - thigh_volume_r (R)"(ComponentPose=@n2, Alpha=0.1, ConstraintWeights_0=1) {
    alphaBoolBlend {blendOption=Linear}
    boneToModify {boneName=thigh_volume_r}
    constraintSetup [{transformType=Rotation, targetBone {boneName=thigh_r}}]
    constraintWeights [1]
   }
   n4 = Get LegLeftUpAlpha() -> (LegLeftUpAlpha:float)
   n5 = "Transform (Modify) Bone - Bone: thigh_volume_l"(ComponentPose=@n3, Alpha=@n4, Translation="-3,-3,0", Rotation="0,0,0", Scale="1.2,1.2,1.2") {
    rotationMode=BMM_Additive
    rotationSpace=BCS_BoneSpace
    scaleMode=BMM_Additive
    scaleSpace=BCS_BoneSpace
    translationMode=BMM_Additive
    translationSpace=BCS_BoneSpace
    alphaBoolBlend {blendOption=Linear}
    alphaScaleBiasClamp {bMapRange=true, inRange {max=0, min=-100}, outRange {max=0, min=1}}
    boneToModify {boneName=thigh_volume_l}
    scale=<1.2 1.2 1.2>
    translation=<-3 -3>
   }
   n6 = "Transform (Modify) Bone - Bone: thigh_volume_r"(ComponentPose=@n5, Alpha=@n4, Translation="3,3,0", Rotation="0,0,0", Scale="1.2,1.2,1.2") {
    rotationMode=BMM_Replace
    rotationSpace=BCS_BoneSpace
    scaleMode=BMM_Additive
    scaleSpace=BCS_BoneSpace
    translationMode=BMM_Additive
    translationSpace=BCS_BoneSpace
    alphaBoolBlend {blendOption=Linear}
    alphaScaleBiasClamp {bMapRange=true, inRange {max=0, min=-100}, outRange {max=0, min=1}}
    boneToModify {boneName=thigh_volume_r}
    scale=<1.2 1.2 1.2>
    translation=<3 3>
   }
   n7 = "Constraint - dyn_kneepad_base_l (PP)"!(ComponentPose=@n6, Alpha=1, ConstraintWeights_0=0.5, ConstraintWeights_1:float=0.5) -> (Pose) {
    alphaBoolBlend {blendOption=Linear}
    boneToModify {boneName=dyn_kneepad_base_l}
    constraintSetup [{targetBone {boneName=thigh_l}}, {targetBone {boneName=calf_l}}]
    constraintWeights [1,1]
   }
   n8 = "Constraint - dyn_kneepad_base_r (PP)"!(ComponentPose=@n7, Alpha=1, ConstraintWeights_0=0.5, ConstraintWeights_1:float=0.5) -> (Pose) {
    alphaBoolBlend {blendOption=Linear}
    boneToModify {boneName=dyn_kneepad_base_r}
    constraintSetup [{targetBone {boneName=thigh_r}}, {targetBone {boneName=calf_r}}]
    constraintWeights [1,1]
   }
   n9 = "Constraint - dyn_hornBelt_a_base (PP)"!(ComponentPose=@n8, Alpha=1, ConstraintWeights_0=0.5, ConstraintWeights_1:float=0.5) -> (Pose) {
    alphaBoolBlend {blendOption=Linear}
    boneToModify {boneName=dyn_hornBelt_a_base}
    constraintSetup [{targetBone {boneName=pelvis}}, {targetBone {boneName=spine_01}}]
    constraintWeights [1,1]
   }
   n10 = "Constraint - dyn_torsoHornA_base_l (PPPP)"!(
    ComponentPose=@n9
    Alpha=1
    ConstraintWeights_0=0.162
    ConstraintWeights_1:float=0.0298
    ConstraintWeights_2:float=0.573
    ConstraintWeights_3:float=0.2352
   ) -> (Pose) {
    alphaBoolBlend {blendOption=Linear}
    boneToModify {boneName=dyn_torsoHornA_base_l}
    constraintSetup [{targetBone {boneName=spine_01}}, {targetBone {boneName=spine_02}}, {targetBone {boneName=dyn_hornBelt_b}}, {targetBone {boneName=pelvis}}]
    constraintWeights [1,1,1,1]
   }
   n11 = "Constraint - dyn_torsoHornB_base_l (PPPP)"!(
    ComponentPose=@n10
    Alpha=1
    ConstraintWeights_0=0.1231
    ConstraintWeights_1:float=0.0203
    ConstraintWeights_2:float=0.6666
    ConstraintWeights_3:float=0.19
   ) -> (Pose) {
    alphaBoolBlend {blendOption=Linear}
    boneToModify {boneName=dyn_torsoHornB_base_l}
    constraintSetup [{targetBone {boneName=spine_01}}, {targetBone {boneName=spine_02}}, {targetBone {boneName=dyn_hornBelt_b}}, {targetBone {boneName=pelvis}}]
    constraintWeights [1,1,1,1]
   }
   n12 = "Constraint - dyn_torsoHornC_base_l (PPP)"!(ComponentPose=@n11, Alpha=1, ConstraintWeights_0=0.068, ConstraintWeights_1:float=0.7548, ConstraintWeights_2:float=0.1772) -> (Pose) {
    alphaBoolBlend {blendOption=Linear}
    boneToModify {boneName=dyn_torsoHornC_base_l}
    constraintSetup [{targetBone {boneName=spine_01}}, {targetBone {boneName=dyn_hornBelt_b}}, {targetBone {boneName=pelvis}}]
    constraintWeights [1,1,1]
   }
   n13 = "Constraint - dyn_torsoHornD_base_l (PPP)"!(ComponentPose=@n12, Alpha=1, ConstraintWeights_0=0.0979, ConstraintWeights_1:float=0.5934, ConstraintWeights_2:float=0.3087) -> (Pose) {
    alphaBoolBlend {blendOption=Linear}
    boneToModify {boneName=dyn_torsoHornD_base_l}
    constraintSetup [{targetBone {boneName=spine_01}}, {targetBone {boneName=dyn_hornBelt_b}}, {targetBone {boneName=pelvis}}]
    constraintWeights [1,1,1]
   }
   n14 = "Constraint - dyn_torsoHornE_base_l (PPP)"!(ComponentPose=@n13, Alpha=1, ConstraintWeights_0=0.127, ConstraintWeights_1:float=0.4719, ConstraintWeights_2:float=0.401) -> (Pose) {
    alphaBoolBlend {blendOption=Linear}
    boneToModify {boneName=dyn_torsoHornE_base_l}
    constraintSetup [{targetBone {boneName=spine_01}}, {targetBone {boneName=dyn_hornBelt_b}}, {targetBone {boneName=pelvis}}]
    constraintWeights [1,1,1]
   }
   n15 = Get LegLeftUpAlpha()
   n16 = "Transform (Modify) Bone - Bone: dyn_hornBelt_a_base"(ComponentPose=@n14, Alpha=@n15, Translation="0,0,0", Rotation="20,0,0", Scale="1,1,1") {
    rotationMode=BMM_Additive
    rotationSpace=BCS_BoneSpace
    scaleMode=BMM_Additive
    scaleSpace=BCS_BoneSpace
    translationMode=BMM_Additive
    translationSpace=BCS_BoneSpace
    alphaBoolBlend {blendOption=Linear}
    alphaScaleBiasClamp {bMapRange=true, inRange {max=0, min=-100}, outRange {max=0, min=1}}
    boneToModify {boneName=dyn_hornBelt_a_base}
    rotation {pitch=20}
   }
   n17 = "Transform (Modify) Bone - Bone: dyn_torsoHornA_base_l"(ComponentPose=@n16, Alpha=@n15, Translation="0,0,0", Rotation="-25,0,0", Scale="1,1,1") {
    rotationMode=BMM_Additive
    rotationSpace=BCS_BoneSpace
    scaleMode=BMM_Additive
    scaleSpace=BCS_BoneSpace
    translationMode=BMM_Additive
    translationSpace=BCS_BoneSpace
    alphaBoolBlend {blendOption=Linear}
    alphaScaleBiasClamp {bMapRange=true, inRange {max=0, min=-100}, outRange {max=0, min=1}}
    boneToModify {boneName=dyn_torsoHornA_base_l}
    rotation {pitch=-25}
   }
   n18 = "Transform (Modify) Bone - Bone: dyn_torsoHornB_base_l"(ComponentPose=@n17, Alpha=@n15, Translation="0,0,2", Rotation="45,-10,0", Scale="1,1,1") {
    rotationMode=BMM_Additive
    rotationSpace=BCS_BoneSpace
    scaleMode=BMM_Additive
    scaleSpace=BCS_BoneSpace
    translationMode=BMM_Additive
    translationSpace=BCS_BoneSpace
    alphaBoolBlend {blendOption=Linear}
    alphaScaleBiasClamp {bMapRange=true, inRange {max=0, min=-100}, outRange {max=0, min=1}}
    boneToModify {boneName=dyn_torsoHornB_base_l}
    rotation {pitch=45, yaw=-10}
    translation {z=2}
   }
   n19 = "Transform (Modify) Bone - Bone: dyn_torsoHornC_base_l"(ComponentPose=@n18, Alpha=@n15, Translation="0,0,2", Rotation="20,-15,0", Scale="1,1,1") {
    rotationMode=BMM_Additive
    rotationSpace=BCS_BoneSpace
    scaleMode=BMM_Additive
    scaleSpace=BCS_BoneSpace
    translationMode=BMM_Additive
    translationSpace=BCS_BoneSpace
    alphaBoolBlend {blendOption=Linear}
    alphaScaleBiasClamp {bMapRange=true, inRange {max=0, min=-100}, outRange {max=0, min=1}}
    boneToModify {boneName=dyn_torsoHornC_base_l}
    rotation {pitch=20, yaw=-15}
    translation {z=2}
   }
   n20 = "Transform (Modify) Bone - Bone: dyn_torsoHornD_base_l"(ComponentPose=@n19, Alpha=@n15, Translation="0,0,1", Rotation="5,-10,0", Scale="1,1,1") {
    rotationMode=BMM_Additive
    rotationSpace=BCS_BoneSpace
    scaleMode=BMM_Additive
    scaleSpace=BCS_BoneSpace
    translationMode=BMM_Additive
    translationSpace=BCS_BoneSpace
    alphaBoolBlend {blendOption=Linear}
    alphaScaleBiasClamp {bMapRange=true, inRange {max=0, min=-100}, outRange {max=0, min=1}}
    boneToModify {boneName=dyn_torsoHornD_base_l}
    rotation {pitch=5, yaw=-10}
    translation {z=1}
   }
   n21 = Get LegLeftSideAlpha() -> (LegLeftSideAlpha:float)
   n22 = "Transform (Modify) Bone - Bone: dyn_torsoHornB_base_l"(ComponentPose=@n20, Alpha=@n21, Translation="0,0,0", Rotation="20,15,0", Scale="1,1,1") {
    rotationMode=BMM_Additive
    rotationSpace=BCS_BoneSpace
    scaleMode=BMM_Additive
    scaleSpace=BCS_BoneSpace
    translationMode=BMM_Additive
    translationSpace=BCS_BoneSpace
    alphaBoolBlend {blendOption=Linear}
    alphaScaleBiasClamp {bMapRange=true, inRange {max=20}}
    boneToModify {boneName=dyn_torsoHornB_base_l}
    rotation {pitch=20, yaw=15}
   }
   n23 = "Transform (Modify) Bone - Bone: dyn_torsoHornC_base_l"(ComponentPose=@n22, Alpha=@n21, Translation="0,0,0", Rotation="20,15,0", Scale="1,1,1") {
    rotationMode=BMM_Additive
    rotationSpace=BCS_BoneSpace
    scaleMode=BMM_Additive
    scaleSpace=BCS_BoneSpace
    translationMode=BMM_Additive
    translationSpace=BCS_BoneSpace
    alphaBoolBlend {blendOption=Linear}
    alphaScaleBiasClamp {bMapRange=true, inRange {max=20}}
    boneToModify {boneName=dyn_torsoHornC_base_l}
    rotation {pitch=20, yaw=15}
   }
   n24 = "Transform (Modify) Bone - Bone: dyn_torsoHornD_base_l"(ComponentPose=@n23, Alpha=@n21, Translation="0,0,0", Rotation="10,5,0", Scale="1,1,1") {
    rotationMode=BMM_Additive
    rotationSpace=BCS_BoneSpace
    scaleMode=BMM_Additive
    scaleSpace=BCS_BoneSpace
    translationMode=BMM_Additive
    translationSpace=BCS_BoneSpace
    alphaBoolBlend {blendOption=Linear}
    alphaScaleBiasClamp {bMapRange=true, inRange {max=20}}
    boneToModify {boneName=dyn_torsoHornD_base_l}
    rotation {pitch=10, yaw=5}
   }
   n25 = "Constraint - dyn_scarf_base_base_l (PPPP)"!(
    ComponentPose=@n24
    Alpha=1
    ConstraintWeights_0=0.0488
    ConstraintWeights_1:float=0.0145
    ConstraintWeights_2:float=0.2739
    ConstraintWeights_3:float=0.6628
   ) -> (Pose) {
    alphaBoolBlend {blendOption=Linear}
    boneToModify {boneName=dyn_scarf_base_base_l}
    constraintSetup [{targetBone {boneName=spine_03}}, {targetBone {boneName=spine_04}}, {targetBone {boneName=spine_05}}, {targetBone {boneName=clavicle_l}}]
    constraintWeights [1,1,1,1]
   }
   n26 = "Transform (Modify) Bone - Bone: dyn_scarf_base_base_l"(ComponentPose=@n25, Alpha=Get ShoulderLeftUpAlpha() -> (ShoulderLeftUpAlpha:float), Translation="0,2,0", Rotation="-40,0,0", Scale="1,1,1") {
    rotationMode=BMM_Additive
    rotationSpace=BCS_BoneSpace
    scaleMode=BMM_Additive
    scaleSpace=BCS_BoneSpace
    translationMode=BMM_Additive
    translationSpace=BCS_BoneSpace
    alphaBoolBlend {blendOption=Linear}
    alphaScaleBiasClamp {bMapRange=true, inRange {max=40}}
    boneToModify {boneName=dyn_scarf_base_base_l}
    rotation {pitch=-40}
    translation {y=2}
   }
   n27 = "Constraint - dyn_clothTop_a (PPPPP)"!(
    ComponentPose=@n26
    Alpha=1
    ConstraintWeights_0=0.3543
    ConstraintWeights_1:float=0.2738
    ConstraintWeights_2:float=0.1077
    ConstraintWeights_3:float=0.0151
    ConstraintWeights_4:float=0.2491
   ) -> (Pose) {
    alphaBoolBlend {blendOption=Linear}
    boneToModify {boneName=dyn_clothTop_a}
    constraintSetup [
     {targetBone {boneName=spine_01}}
     {targetBone {boneName=spine_02}}
     {targetBone {boneName=spine_03}}
     {targetBone {boneName=spine_04}}
     {targetBone {boneName=pelvis}}
    ]
    constraintWeights [1,1,1,1,1]
   }
   n28 = "Constraint - dyn_clothTop_b (PPPPP)"!(
    ComponentPose=@n27
    Alpha=1
    ConstraintWeights_0=0.3134
    ConstraintWeights_1:float=0.3185
    ConstraintWeights_2:float=0.1798
    ConstraintWeights_3:float=0.0144
    ConstraintWeights_4:float=0.1739
   ) -> (Pose) {
    alphaBoolBlend {blendOption=Linear}
    boneToModify {boneName=dyn_clothTop_b}
    constraintSetup [
     {targetBone {boneName=spine_01}}
     {targetBone {boneName=spine_02}}
     {targetBone {boneName=spine_03}}
     {targetBone {boneName=spine_04}}
     {targetBone {boneName=pelvis}}
    ]
    constraintWeights [1,1,1,1,1]
   }
   n29 = "Constraint - dyn_clothTop_c (PPPP)"!(
    ComponentPose=@n28
    Alpha=1
    ConstraintWeights_0=0.3403
    ConstraintWeights_1:float=0.2988
    ConstraintWeights_2:float=0.1385
    ConstraintWeights_3:float=0.2224
   ) -> (Pose) {
    alphaBoolBlend {blendOption=Linear}
    boneToModify {boneName=dyn_clothTop_c}
    constraintSetup [{targetBone {boneName=spine_01}}, {targetBone {boneName=spine_02}}, {targetBone {boneName=spine_03}}, {targetBone {boneName=pelvis}}]
    constraintWeights [1,1,1,1]
   }
   n30 = "Constraint - dyn_clothTop_d (PPPPP)"!(
    ComponentPose=@n29
    Alpha=1
    ConstraintWeights_0=0.3384
    ConstraintWeights_1:float=0.2271
    ConstraintWeights_2:float=0.1238
    ConstraintWeights_3:float=0.0192
    ConstraintWeights_4:float=0.2914
   ) -> (Pose) {
    alphaBoolBlend {blendOption=Linear}
    boneToModify {boneName=dyn_clothTop_d}
    constraintSetup [
     {targetBone {boneName=spine_01}}
     {targetBone {boneName=spine_02}}
     {targetBone {boneName=spine_03}}
     {targetBone {boneName=spine_04}}
     {targetBone {boneName=pelvis}}
    ]
    constraintWeights [1,1,1,1,1]
   }
   n31 = "Constraint - dyn_clothTop_e (PPPP)"!(
    ComponentPose=@n30
    Alpha=1
    ConstraintWeights_0=0.3854
    ConstraintWeights_1:float=0.1669
    ConstraintWeights_2:float=0.065
    ConstraintWeights_3:float=0.3828
   ) -> (Pose) {
    alphaBoolBlend {blendOption=Linear}
    boneToModify {boneName=dyn_clothTop_e}
    constraintSetup [{targetBone {boneName=spine_01}}, {targetBone {boneName=spine_02}}, {targetBone {boneName=spine_03}}, {targetBone {boneName=pelvis}}]
    constraintWeights [1,1,1,1]
   }
   n32 = Anim Dynamics - Bone: dyn_clothTop_a(ComponentPose=@n31, Alpha=1, ExternalForce="0,0,0") {
    angularDampingOverride=1
    angularSpringConstant=1500
    bAngularSpring=true
    bGravityOverrideInSimSpace=true
    bOverrideAngularDamping=true
    gravityScale=0.35
    simulationSpace=RootRelative
    alphaBoolBlend {blendOption=Linear}
    boundBone {boneName=dyn_clothTop_a}
    componentLinearAccScale=<0.3 0.3 0.5>
    componentLinearVelScale=<0.3 0.3 0.5>
    physicsBodyDefinitions [{
      sphereCollisionRadius=0
      boundBone {boneName=dyn_clothTop_a}
      boxExtents=<5 25 25>
      constraintSetup {angularLimitsMax=<0 5 15>, angularLimitsMin=<0 -5 -10>, angularTarget=<1 0 0>}
      localJointOffset=<10 0 0>
     }]
   }
   n33 = Anim Dynamics - Bone: dyn_clothTop_b(ComponentPose=@n32, Alpha=1, ExternalForce="0,0,0") {
    angularDampingOverride=1
    angularSpringConstant=1500
    bAngularSpring=true
    bGravityOverrideInSimSpace=true
    bOverrideAngularDamping=true
    gravityScale=0.35
    simulationSpace=RootRelative
    alphaBoolBlend {blendOption=Linear}
    boundBone {boneName=dyn_clothTop_b}
    componentLinearAccScale=<0.3 0.3 0.5>
    componentLinearVelScale=<0.3 0.3 0.5>
    physicsBodyDefinitions [{
      sphereCollisionRadius=0
      boundBone {boneName=dyn_clothTop_b}
      boxExtents=<10 20 20>
      constraintSetup {angularLimitsMax=<0 0 0>, angularLimitsMin=<0 -15 0>, angularTarget=<1 0 0>}
      localJointOffset=<10 0 0>
     }]
   }
   n34 = Anim Dynamics - Bone: dyn_clothTop_c(ComponentPose=@n33, Alpha=1, ExternalForce="0,0,0") {
    angularDampingOverride=1
    angularSpringConstant=1500
    bAngularSpring=true
    bGravityOverrideInSimSpace=true
    bOverrideAngularDamping=true
    gravityScale=0.35
    simulationSpace=RootRelative
    alphaBoolBlend {blendOption=Linear}
    boundBone {boneName=dyn_clothTop_c}
    componentLinearAccScale=<0.3 0.3 0.5>
    componentLinearVelScale=<0.3 0.3 0.5>
    physicsBodyDefinitions [{
      sphereCollisionRadius=0
      boundBone {boneName=dyn_clothTop_c}
      boxExtents=<10 25 25>
      constraintSetup {angularLimitsMax=<0 0 0>, angularLimitsMin=<0 -30 0>, angularTarget=<1 0 0>}
      localJointOffset=<3 0 0>
     }]
   }
   n35 = Anim Dynamics - Bone: dyn_clothTop_d(ComponentPose=@n34, Alpha=1, ExternalForce="0,0,0") {
    angularDampingOverride=1
    angularSpringConstant=1500
    bAngularSpring=true
    bGravityOverrideInSimSpace=true
    bOverrideAngularDamping=true
    gravityScale=0.35
    simulationSpace=RootRelative
    alphaBoolBlend {blendOption=Linear}
    boundBone {boneName=dyn_clothTop_d}
    componentLinearAccScale=<0.3 0.3 0.5>
    componentLinearVelScale=<0.3 0.3 0.5>
    physicsBodyDefinitions [{
      sphereCollisionRadius=0
      boundBone {boneName=dyn_clothTop_d}
      boxExtents=<10 25 25>
      constraintSetup {angularLimitsMax=<0 0 20>, angularLimitsMin=<0 0 0>, angularTarget=<1 0 0>}
      localJointOffset=<3 0 0>
     }]
   }
   n36 = Anim Dynamics - Bone: dyn_clothTop_e(ComponentPose=@n35, Alpha=1, ExternalForce="0,0,0") {
    angularDampingOverride=1
    angularSpringConstant=1500
    bAngularSpring=true
    bGravityOverrideInSimSpace=true
    bOverrideAngularDamping=true
    gravityScale=0.35
    simulationSpace=RootRelative
    alphaBoolBlend {blendOption=Linear}
    boundBone {boneName=dyn_clothTop_e}
    componentLinearAccScale=<0.3 0.3 0.5>
    componentLinearVelScale=<0.3 0.3 0.5>
    physicsBodyDefinitions [{
      sphereCollisionRadius=0
      boundBone {boneName=dyn_clothTop_e}
      boxExtents=<10 25 25>
      constraintSetup {angularLimitsMax=<0 0 15>, angularLimitsMin=<0 0 0>, angularTarget=<1 0 0>}
      localJointOffset=<5 0 0>
     }]
   }
   n37 = "Constraint - dyn_clothBottomA_a_base (R)"(ComponentPose=@n36, Alpha=1, ConstraintWeights_0=1) {
    alphaBoolBlend {blendOption=Linear}
    boneToModify {boneName=dyn_clothBottomA_a_base}
    constraintSetup [{transformType=Rotation, targetBone {boneName=thigh_r}}]
    constraintWeights [1]
   }
   n38 = "Constraint - dyn_clothBottomB_a_base (R)"(ComponentPose=@n37, Alpha=1, ConstraintWeights_0=1) {
    alphaBoolBlend {blendOption=Linear}
    boneToModify {boneName=dyn_clothBottomB_a_base}
    constraintSetup [{transformType=Rotation, targetBone {boneName=thigh_r}}]
    constraintWeights [1]
   }
   n39 = Get LegRightUpAlpha() -> (LegRightUpAlpha:float)
   n40 = "Transform (Modify) Bone - Bone: dyn_clothBottomA_a_base"(ComponentPose=@n38, Alpha=@n39, Translation="0,1,0", Rotation="0,0,0", Scale="1,1,1") {
    rotationMode=BMM_Additive
    rotationSpace=BCS_BoneSpace
    scaleMode=BMM_Additive
    scaleSpace=BCS_BoneSpace
    translationMode=BMM_Additive
    translationSpace=BCS_BoneSpace
    alphaBoolBlend {blendOption=Linear}
    alphaScaleBiasClamp {bMapRange=true, inRange {max=0, min=-100}, outRange {max=0, min=1}}
    boneToModify {boneName=dyn_clothBottomA_a_base}
    translation {y=1}
   }
   n41 = "Transform (Modify) Bone - Bone: dyn_clothBottomB_a_base"(ComponentPose=@n40, Alpha=@n39, Translation="0,4,0", Rotation="0,0,0", Scale="1,1,1") {
    rotationMode=BMM_Additive
    rotationSpace=BCS_BoneSpace
    scaleMode=BMM_Additive
    scaleSpace=BCS_BoneSpace
    translationMode=BMM_Additive
    translationSpace=BCS_BoneSpace
    alphaBoolBlend {blendOption=Linear}
    alphaScaleBiasClamp {bMapRange=true, inRange {max=0, min=-100}, outRange {max=0, min=1}}
    boneToModify {boneName=dyn_clothBottomB_a_base}
    translation {y=4}
   }
   n42 = RigidBody(ComponentPose=@n41) {
    bClampLinearTranslationLimitToRefPose=true
    bOverrideWorldGravity=true
    overridePhysicsAsset=&/M_MED_Bone_Snake_Physics
    simulationSpace=BaseBoneSpace
    alphaBoolBlend {blendOption=Linear}
    baseBoneRef {boneName=root}
    componentLinearAccScale=<0.4 0.4 0.45>
    componentLinearVelScale=<0.4 0.4 0.45>
    overrideWorldGravity {z=-980}
   }
   n43 = Get IsBackpackOn() -> (IsBackpackOn:double)
   n44 = "Constraint - dyn_scarf_back_a (PPPPP)"!(
    ComponentPose=@n42
    Alpha=@n43
    ConstraintWeights_0=0.0787
    ConstraintWeights_1:float=0.0314
    ConstraintWeights_2:float=0.0336
    ConstraintWeights_3:float=0.7769
    ConstraintWeights_4:float=0.0794
   ) -> (Pose) {
    alphaBoolBlend {blendOption=Linear}
    boneToModify {boneName=dyn_scarf_back_a}
    constraintSetup [
     {targetBone {boneName=clavicle_r}}
     {targetBone {boneName=spine_03}}
     {targetBone {boneName=spine_04}}
     {targetBone {boneName=spine_05}}
     {targetBone {boneName=clavicle_l}}
    ]
    constraintWeights [1,1,1,1,1]
   }
   n45 = "Constraint - dyn_scarf_back_b (PPPPPP)"!(
    ComponentPose=@n44
    Alpha=@n43
    ConstraintWeights_0=0.1666
    ConstraintWeights_1:float=0.0098
    ConstraintWeights_2:float=0.2783
    ConstraintWeights_3:float=0.1416
    ConstraintWeights_4:float=0.216
    ConstraintWeights_5:float=0.1877
   ) -> (Pose) {
    alphaBoolBlend {blendOption=Linear}
    boneToModify {boneName=dyn_scarf_back_b}
    constraintSetup [
     {targetBone {boneName=clavicle_l}}
     {targetBone {boneName=spine_02}}
     {targetBone {boneName=spine_03}}
     {targetBone {boneName=spine_04}}
     {targetBone {boneName=spine_05}}
     {targetBone {boneName=clavicle_r}}
    ]
    constraintWeights [1,1,1,1,1,1]
   }
   n46 = "Transform (Modify) Bone - Bone: dyn_scarf_front_a_base"(ComponentPose=@n45, Alpha=1, Translation="0,2,0", Rotation="0,0,0", Scale="1,1,1") {
    rotationMode=BMM_Additive
    rotationSpace=BCS_BoneSpace
    scaleMode=BMM_Additive
    scaleSpace=BCS_BoneSpace
    translationMode=BMM_Additive
    translationSpace=BCS_BoneSpace
    alphaBoolBlend {blendOption=Linear}
    alphaScaleBiasClamp {inRange {max=20}}
    boneToModify {boneName=dyn_scarf_front_a_base}
    translation {y=2}
   }
   n47 = "Transform (Modify) Bone - Bone: dyn_clothBottomD_a_base"(ComponentPose=@n46, Alpha=1, Translation="0,0,1", Rotation="0,0,0", Scale="1,1,1") {
    rotationMode=BMM_Additive
    rotationSpace=BCS_BoneSpace
    scaleMode=BMM_Additive
    scaleSpace=BCS_BoneSpace
    translationMode=BMM_Additive
    translationSpace=BCS_BoneSpace
    alphaBoolBlend {blendOption=Linear}
    alphaScaleBiasClamp {inRange {max=20}}
    boneToModify {boneName=dyn_clothBottomD_a_base}
    translation {z=1}
   }
   Output Pose(Result=Component To Local(ComponentPose=@n47)) {layerGroup=None}
  ]}
 EventGraph {links [], nodes []}
]
variables [IsBackpackOn {type=number}]
