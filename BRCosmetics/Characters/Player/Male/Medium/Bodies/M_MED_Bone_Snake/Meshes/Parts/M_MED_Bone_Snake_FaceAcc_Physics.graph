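# Physics asset for the face accessory mesh of the male Bone Snake player cosmetic, simulating a dangling tongue (bones named "toung").
# Four capsule bodies form a chain from the base bone through segments a, b and c, with the tip segment the longest.
# Three constraints link each segment to the previous one with tight, identical swing and twist limits, so the tongue flexes only slightly at each joint.
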PhysicsAsset /BRCosmetics/Characters/Player/Male/Medium/Bodies/M_MED_Bone_Snake/Meshes/Parts/M_MED_Bone_Snake_FaceAcc_Physics
bodies {
 dyn_toung_a [{shapeType=Capsule, length=2, radius=1, center=<2.4602 -0.0001 0>, rotation=<rot 90 0 0>}]
 dyn_toung_a_base [{shapeType=Capsule, length=1, radius=1}]
 dyn_toung_b [{shapeType=Capsule, length=2.25, radius=1, center=<2.3677 0 0>, rotation=<rot 90 0 0>}]
 dyn_toung_c [{shapeType=Capsule, length=3.75, radius=1, center=<2.7681 0 0>, rotation=<rot 90 0 0>}]
}
constraints [
 {bone1Name=dyn_toung_a, bone2Name=dyn_toung_a_base, swing1LimitDegrees=10, swing2LimitDegrees=5, twistLimitDegrees=5}
 {bone1Name=dyn_toung_b, bone2Name=dyn_toung_a, swing1LimitDegrees=10, swing2LimitDegrees=5, twistLimitDegrees=5}
 {bone1Name=dyn_toung_c, bone2Name=dyn_toung_b, swing1LimitDegrees=10, swing2LimitDegrees=5, twistLimitDegrees=5}
]
